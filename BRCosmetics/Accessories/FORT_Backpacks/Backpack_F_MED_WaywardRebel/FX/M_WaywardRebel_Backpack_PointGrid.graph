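# Additive unlit material for Niagara mesh particles on the WaywardRebel backpack, drawing a halftone-like dot grid glow.
# A vertically panning multicolor texture is tinted by particle color, weighted by how much the surface faces the camera, and fed through an emissive adjustment function.
# Screen-aligned coordinates centered on particle position are rotated and tiled into dots.
# These dots combine with a soft facing term, then depth fade and particle alpha shape opacity.
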Material /BRCosmetics/Accessories/FORT_Backpacks/Backpack_F_MED_WaywardRebel/FX/M_WaywardRebel_Backpack_PointGrid
graph {
 properties {
  bUsedWithNiagaraMeshParticles=true
  blendMode=BLEND_Additive
  pixelDepthOffsetMode=PDOM_Legacy
  shadingModel=MSM_Unlit
  usedShadingModels=Unlit
 }
 nodes [
  n1 = Panner(Coordinate=TextureCoordinate(), Time=Multiply(A=Time(), B=ScalarParameter() {defaultValue=1, parameterName=ColorTexture_Speed})) {speedY=0.5}
  n2 = DotProduct(A=VertexNormalWS(), B=CameraVectorWS())
  n3 = Multiply(A=ParticleColor().RGB, B=@n2)
  n4 = Saturate(None=@n2)
  MaterialFunctionCall(Emissive=Multiply(
    A=TextureSampleParameter2D(UVs=@n1) {parameterName=ColorTexture, texture=/BRCosmetics/Weapons/FORT_Melee/Pickaxe_SpyMaleFNCS1H/FX/T_FX_SpyMaleFNCS_MultiColor}.RGB
    B=Multiply(A=@n3, B=@n4)
   )) {materialFunction=&MF_FX_EmissiveAdjustments}
  n5 = NamedRerouteUsage()
  n6 = Add(A=Multiply(A=Power(Base=@n2) {constExponent=3}) {constB=0.25}, B=SmoothStep(Value=Multiply(A=@n2, B=@n5)) {constMin=0.5})
  n7 = DepthFade(FadeDistance=ScalarParameter() {defaultValue=15, parameterName=DepthFade}) {fadeDistanceDefault=15}
  Multiply(A=Saturate(None=Multiply(A=@n6, B=@n7)), B=ParticleColor().A)
  n8 = ParticlePositionWS()
  n9 = Multiply(
   A=Divide(A=MaterialFunctionCall!(Center Point World Position=@n8) {materialFunction=/Engine/Functions/Engine_MaterialFunctions02/UVs/WorldSpaceAlignedScreenCoordinates}.Offset UVs) {constB=100}
   B=ScalarParameter() {defaultValue=40, parameterName=Dots}
  ) {constA=0.5}
  n10 = Constant() {r=1}
  n11 = Add(A=Frac(None=Rotator(Coordinate=@n9, Time=@n10) {speed=0.75})) {constB=-0.5}
  n12 = Constant() {r=0.1}
  NamedRerouteDeclaration(None=OneMinus(None=Distance(A=@n11, B=@n12))) {name=DotBlend, nodeColor=<rgba 0 0.4844 1 1>}
 ]
}
