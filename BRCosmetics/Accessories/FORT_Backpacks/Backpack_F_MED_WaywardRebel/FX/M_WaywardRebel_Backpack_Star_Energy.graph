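# Translucent, unlit, two-sided Niagara sprite and mesh particle material for star energy effects on the WaywardRebel backpack, with SplineThicken support.
# Use_SplineThicken blends noise-modulated curve-atlas color with a two-color gradient built from sphere masks and radially panned noise.
# Opacity combines a star texture mask, dynamic-parameter sphere-mask dissolve (sine-shaped or noise-distorted), view-angle fade and depth fade.
# Particle color tints the emissive, with optional dynamic desaturation.
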Material /BRCosmetics/Accessories/FORT_Backpacks/Backpack_F_MED_WaywardRebel/FX/M_WaywardRebel_Backpack_Star_Energy
graph {
 properties {
  bUsedWithNiagaraMeshParticles=true
  bUsedWithNiagaraSprites=true
  blendMode=BLEND_Translucent
  pixelDepthOffsetMode=PDOM_Legacy
  shadingModel=MSM_Unlit
  twoSided=true
  usedShadingModels=Unlit
 }
 nodes [
  n1 = ParticleColor()
  n2 = TextureCoordinate()
  n3 = Multiply(
   A=StaticComponentMaskParameter(None=TextureSampleParameter2D(UVs=@n2) {parameterName=MaskTex, texture=/Game/Accessories/FORT_Backpacks/Backpack_F_MED_WaywardRebel/FX/T_WaywardRebel_Backpack_StarRGB}.RGB) {defaultR=true, parameterName=ChannelMask}
   B=ScalarParameter() {defaultValue=2.5, parameterName=Mask_Emissive}
  )
  n4 = ScalarParameter() {defaultValue=1.5, parameterName=Mask_power}
  n5 = ScalarParameter() {defaultValue=1, group=Use_SplineThicken, parameterName=Use_SplineThicken}
  n6 = Multiply(A=@n1.A, B=LinearInterpolate(A=@n1.A, B=Power(Base=@n3, Exp=@n4), Alpha=@n5))
  n7 = Sine(None=ComponentMask(None=TextureCoordinate()) {g=true}) {period=2}
  n8 = ScalarParameter() {defaultValue=1, parameterName=SplineThicken_MaskPower}
  n9 = TextureCoordinate() {coordinateIndex=1}
  n10 = Constant2Vector() {g=0.5, r=0.5}
  n11 = ScalarParameter() {defaultValue=0.6, group=Color, parameterName=Hardness03}
  n12 = DynamicParameter() {defaultValue=<rgba 0 1 1 1>, paramNames [Dissilve_Radius03, Desaturation, Param3, Param4]}
  n13 = Multiply(A=Multiply(A=@n6, B=Power(Base=@n7, Exp=@n8)), B=OneMinus(None=SphereMask(A=@n9, B=@n10.G, Radius=@n12.Dissilve_Radius03, Hardness=@n11)))
  n14 = TextureCoordinate()
  n15 = AppendVector(A=ScalarParameter() {defaultValue=1, parameterName=Texture_A_UTiling}, B=ScalarParameter() {defaultValue=1, parameterName=Texture_A_VTiling})
  n16 = MaterialFunctionCall() {materialFunction=&VectorToRadialValue}
  n17 = AppendVector(A=ScalarParameter() {defaultValue=2, parameterName=Noise_UTiling}, B=ScalarParameter() {defaultValue=0.5, parameterName=Noise_VTiling})
  n18 = TextureSample(UVs=Panner(Coordinate=Multiply(A=@n16.Radial Coordinates, B=@n17)) {speedY=-0.2}) {samplerType=SAMPLERTYPE_Grayscale, texture=&LowResBlurredNoise}
  n19 = SphereMask(
   A=Add(A=Multiply(A=@n14, B=@n15), B=Multiply(A=@n18.R) {constB=0.05})
   B=Constant2Vector() {g=0.5, r=0.5}.G
   Radius=@n12.Dissilve_Radius03
   Hardness=ScalarParameter() {defaultValue=0.6, group=Color, parameterName=Hardness03}
  )
  n20 = Multiply(A=@n6, B=Power(Base=OneMinus(None=@n19), Exp=ScalarParameter() {defaultValue=1, group=Color, parameterName=Color_power}))
  n21 = ScalarParameter() {defaultValue=1, parameterName=MeshFade}
  DepthFade(
   Opacity=Multiply(
    A=LinearInterpolate(A=@n13, B=@n20, Alpha=@n5)
    B=MaterialFunctionCall("FadeSensitivity (optional)"=@n21) {materialFunction=&FadeBasedOnViewAngle}.Result
   )
   FadeDistance=ScalarParameter() {defaultValue=50, parameterName=Depth Fade}
  )
  n22 = VectorParameter() {group=Color, parameterName=Color01, defaultValue=<rgba 1 0.1222 0.0364 1>}
  n23 = VectorParameter() {group=Color, parameterName=Color02, defaultValue=<rgba 1 0.9425 0.2891 1>}
  n24 = SphereMask(
   A=TextureCoordinate()
   B=Constant2Vector() {g=0.5, r=0.5}.G
   Radius=ScalarParameter() {defaultValue=0.42, group=Color, parameterName=Radius}
   Hardness=ScalarParameter() {defaultValue=0.6, group=Color, parameterName=Hardness}
  )
  n25 = ScalarParameter() {defaultValue=1, group=Color, parameterName=Color_power}
  n26 = AppendVector(A=ScalarParameter() {defaultValue=1, parameterName=U01}, B=ScalarParameter() {defaultValue=1, parameterName=V01})
  n27 = ScalarParameter() {parameterName=Speed_U01}
  n28 = ScalarParameter() {defaultValue=-0.2, parameterName=Speed_V01}
  n29 = TextureSampleParameter2D(UVs=Panner(Coordinate=Multiply(A=@n16.Radial Coordinates, B=@n26), Speed=AppendVector(A=@n27, B=@n28))) {parameterName=ColormasknoiseTex, samplerType=SAMPLERTYPE_LinearGrayscale, texture=/Game/Global/Textures/Default/Noise/T_SwirlNoise_N_height}
  n30 = ScalarParameter() {defaultValue=1, parameterName=ColormasknoiseTex_power}
  n31 = Add(
   A=Multiply(A=Power(Base=@n24, Exp=@n25), B=Power(Base=@n29.R, Exp=@n30))
   B=SphereMask(
    A=TextureCoordinate()
    B=Constant2Vector() {g=0.5, r=0.5}.G
    Radius=ScalarParameter() {defaultValue=0.2, group=Color, parameterName=Radius02}
    Hardness=ScalarParameter() {defaultValue=0.6, group=Color, parameterName=Hardness02}
   )
  )
  n32 = Desaturation(
   None=LinearInterpolate(A=NamedRerouteUsage(), B=LinearInterpolate(A=@n22.RGB, B=@n23.RGB, Alpha=@n31), Alpha=@n5)
   Fraction=LinearInterpolate(
    A=ScalarParameter() {parameterName=Desaturation}
    B=DynamicParameter() {defaultValue=<rgba 0 1 1 1>, paramNames [Dissilve_Radius03, Desaturation, Param3, Param4]}.Desaturation
    Alpha=ScalarParameter() {parameterName=Use_Dynamic_Desaturation}
   )
  )
  MaterialFunctionCall(Emissive=Multiply(A=Multiply(A=@n32, B=@n1.RGB), B=ScalarParameter() {defaultValue=1, parameterName=Emissive})) {materialFunction=&MF_FX_EmissiveAdjustments}
  MaterialFunctionCall(
   WidthBase=ScalarParameter() {defaultValue=3, parameterName=Base}
   WidthTip=ScalarParameter() {defaultValue=3, parameterName=Tip}
   Expand U or V UV Channel=StaticBool()
  ) {materialFunction=&SplineThicken}
  MaterialFunctionCall() {materialFunction=&DebugTimeSine}
  n33 = CurveAtlasRowParameter(CurveTime=ComponentMask(None=TextureCoordinate()) {r=true}) {
   atlas=/Game/Accessories/FORT_Backpacks/Backpack_F_MED_WaywardRebel/FX/CAtlas_WaywardRebel_ColorCurve
   curve=/Game/Accessories/FORT_Backpacks/Backpack_F_MED_WaywardRebel/FX/CV_WaywardRebel_ColorCurve
   parameterName=Color
  }
  n34 = Multiply(
   A=MaterialFunctionCall() {materialFunction=&VectorToRadialValue}.Radial Coordinates
   B=AppendVector(A=ScalarParameter() {defaultValue=1, parameterName=SplineThicken_U01}, B=ScalarParameter() {defaultValue=1, parameterName=SplineThicken_V01})
  )
  n35 = Power(Base=TextureSample(UVs=Panner(Coordinate=@n34) {speedY=-0.5}) {samplerType=SAMPLERTYPE_LinearGrayscale, texture=&T_SwirlNoise_N_height}.R) {constExponent=1.5}
  NamedRerouteDeclaration(None=Multiply(A=@n33, B=LinearInterpolate(Alpha=@n35) {constA=0.9, constB=2})) {name=Color lerp, nodeColor=<rgba 0.8084 0.1529 1 1>}
  ScalarParameter() {defaultValue=0.4, group=Color, parameterName=Dissilve_Radius03}
  TextureCoordinate()
  TextureCoordinate()
 ]
}
