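# Translucent, unlit, two-sided Niagara sprite material for the star effects on the Wayward Rebel backpack.
# A channel-selected, sharpened star mask is multiplied by particle color and a tiled layer texture whose UV scale varies per particle, then faded by view angle.
# Use_Triangle blends that toward a sphere-masked triangle texture, and particle alpha scales the result.
# Particle color times an Emissive parameter feeds a shared emissive-adjustment function.
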Material /BRCosmetics/Accessories/FORT_Backpacks/Backpack_F_MED_WaywardRebel/FX/M_WaywardRebel_Backpack_Star_Multiple
graph {
 properties {
  bUsedWithNiagaraSprites=true
  blendMode=BLEND_Translucent
  pixelDepthOffsetMode=PDOM_Legacy
  shadingModel=MSM_Unlit
  twoSided=true
  usedShadingModels=Unlit
 }
 nodes [
  n1 = ParticleColor()
  MaterialFunctionCall(Emissive=Multiply(A=@n1.RGB, B=ScalarParameter() {defaultValue=1, parameterName=Emissive}) {constB=100}) {materialFunction=&MF_FX_EmissiveAdjustments}
  n2 = ParticleRandom()
  n3 = TextureSampleParameter2D(UVs=Multiply(A=TextureCoordinate() {uTiling=4, vTiling=4}, B=LinearInterpolate(Alpha=@n2) {constA=0.7})) {parameterName=LayTexture, texture=&/T_WaywardRebel_Backpack_Star_Size}
  n4 = TextureSampleParameter2D() {parameterName=MaskTexture, texture=/Game/Accessories/FORT_Backpacks/Backpack_F_MED_WaywardRebel/FX/T_WaywardRebel_Backpack_StarRGB}
  n5 = Multiply(A=Power(Base=StaticComponentMaskParameter(None=@n4.RGB) {defaultG=true, parameterName=ChannelMask}) {constExponent=3}) {constB=3}
  n6 = Multiply(A=Multiply(A=@n1.RGB, B=@n3.R), B=Multiply(A=@n3.R, B=Saturate(None=@n5)))
  n7 = MaterialFunctionCall("FadeSensitivity (optional)"=ScalarParameter() {defaultValue=1, parameterName=MeshFade}) {materialFunction=&FadeBasedOnViewAngle}
  n8 = TextureCoordinate()
  n9 = Constant2Vector() {g=0.45, r=0.46}
  n10 = Constant() {r=0.26}
  n11 = Constant() {r=0.01}
  n12 = TextureSampleParameter2D() {parameterName=Texture02, texture=/Game/Effects/Fort_Effects/Effects/Characters/T_Triangle3}
  n13 = Multiply(A=Divide(A=SphereMask(A=@n8, B=@n9, Radius=@n10, Hardness=@n11)) {constB=3}, B=@n12.G)
  n14 = Multiply(A=@n12.G) {constB=0.18}
  Multiply(
   A=ParticleColor().A
   B=LinearInterpolate(A=Multiply(A=@n6, B=@n7.Result), B=Add(A=@n13, B=@n14), Alpha=ScalarParameter() {parameterName=Use_Triangle})
  )
 ]
}
