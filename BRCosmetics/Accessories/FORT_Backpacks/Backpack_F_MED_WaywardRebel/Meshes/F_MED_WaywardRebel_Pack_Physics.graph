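# Physics asset for the Wayward Rebel backpack cosmetic mesh, providing collision and joint limits for its dangling straps.
# The pack bone carries two large capsules, and three strap bones each carry a small capsule.
# Each strap is constrained directly to the pack with identical limits: moderate swing on one axis, wider on the other, and almost no twist.
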PhysicsAsset /BRCosmetics/Accessories/FORT_Backpacks/Backpack_F_MED_WaywardRebel/Meshes/F_MED_WaywardRebel_Pack_Physics
bodies {
 dyn_strap_a [{shapeType=Capsule, length=2, radius=2, center=<4 0 0>, rotation=<rot 90 0 0>}]
 dyn_strap_b [{shapeType=Capsule, length=2, radius=2, center=<4 0 0>, rotation=<rot 90 0 0>}]
 dyn_strap_c [{shapeType=Capsule, length=2, radius=2, center=<9 0 0>, rotation=<rot 80 179.9981 179.9996>}]
 pack [
  {shapeType=Capsule, length=6, radius=16, center=<1.303 -3.3091 3.2384>}
  {shapeType=Capsule, length=17, radius=17, shapeName=pack_capsule, center=<0 15.5485 -26.4142>, rotation=<rot 90 0 0>}
 ]
}
constraints [
 {bone1Name=dyn_strap_a, bone2Name=pack, swing1LimitDegrees=30, swing2LimitDegrees=60, twistLimitDegrees=2}
 {bone1Name=dyn_strap_b, bone2Name=pack, swing1LimitDegrees=30, swing2LimitDegrees=60, twistLimitDegrees=2}
 {bone1Name=dyn_strap_c, bone2Name=pack, swing1LimitDegrees=30, swing2LimitDegrees=60, twistLimitDegrees=2}
]
